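# Niagara idle effect for the TaperSlink mic in Sparks cosmetics, drawing a small warm flame.
# A single burst-spawned FlameInside emitter feeds two mesh renderers, a sprite and a particle light.
# Randomized scale and light intensity add variation, while a system GlobalSize value drives mesh height and sprite size.
# Mesh brightness varies by platform, and a front-end preview user flag switches the initial color alpha.
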NiagaraSystem /SparksCosmetics/Instrument/Mic/TaperSlink/Effects/NS_Mic_TaperSlink_Idle
dependencies {
 usedDataInterfaces [&NiagaraDataInterfacePlatformSet, &NiagaraDataInterfaceArrayFloat3]
 usedDynamicInputs [
  &AddVectorToPosition
  &TransformVector
  &SelectVectorFromArray
  &MakeFloatFromVector
 ]
 usedModules [
  &self:SystemSpawnScript.SystemScriptSource.SystemScriptGraph.NiagaraNodeAssignment_3.SetVariables_#1
  &SystemState
 ]
}
emitters {FlameInside {
  emitterUpdateScript [
   &EmitterState()
   &SpawnBurst_Instantaneous(Use Spawn Probability=false, UseLoopCountLimit=false)
  ]
  particleSpawnScript [
   &InitializeParticle(
    Color=&MakeCustomLinearColorFromBool(Bool=$User.bIsFrontEndPreview, True Color=<rgba 1 1 1 1>, False Color=<rgba 1 1 1 5>)
    Lifetime=0.5
    Mesh Scale=&Multiply_VectorByFloat(Vector=<1 1 1.75>, Float=0.0375)
    Mesh Scale Mode=::NewEnumerator4
    Sprite Size Mode=::NewEnumerator3
    Uniform Sprite Size=20
    UsePositionOffset=true
    Write Material Random=true
   )
   &InitialMeshOrientation(Mesh Axis To Orient=<0 0 1>, Mesh Orientation Mode=::NewEnumerator2, Orientation Vector=<0 0 -1>, Use Rotation Vector=true)
  ]
  particleUpdateScript [
   &ParticleState()
   &SolveForcesAndVelocity(Clamp Velocity=false, Limit Acceleration=false)
   &ScaleMeshSize(Scale Factor=&MakeVector(
     X=&RandomRangeFloat(Minimum=1, Maximum=1.5, Evaluation Type=::NewEnumerator1)
     Y=&RandomRangeFloat(Minimum=1, Maximum=1.5, Evaluation Type=::NewEnumerator1)
     Z=$System.GlobalSize
    ))
   &DynamicMaterialParameters(
    Index 0 Param 1=0.05
    Module.Param0WriteEnabled=true
    Module.Param1WriteEnabled=true
    Module.Param2WriteEnabled=true
    Module.Param3WriteEnabled=true
   )
   &FN_Mesh_Attributes(
    Color=&MultiplyLinearColor(
     Linear Color=<rgba 3 1.2672 0.226 1>
     Scale Factor=&MakeVector4_FromFloat(Float=&FN_PickFloatDependOnPlatform(TargetPlatform=DataInterface() {propertyValues {Platforms {qualityLevelMask=-2}}}, TargetPlatformActive=0.8, TargetPlatformNoActive=0.5))
    )
    Create Color Attribute=true
    Create Scale Attribute=true
    Scale=&Multiply_VectorByFloat(
     Vector=&RandomRangeVector(Minimum=<0.06 0.06 0.12>, Maximum=<0.06 0.06 0.17>, Evaluation Type=::NewEnumerator1)
     Float=0.3
    )
    Write to Module Namespace=true
   )
   &FN_Light_Attributes(
    Color=<rgba 30 8.139 0 1>
    Create Color Attribute=true
    Create Exponent Attribute=true
    Create Radius Attribute=true
    Enable Particle Light Enable=true
    Exponent=3
    Radius=10
    Scale Color=&RandomRangeFloat(Minimum=0.5, Maximum=1, Evaluation Type=::NewEnumerator1)
   )
   &FN_Sprite_Attributes(
    Color=<rgba 0.3726 0 0 1>
    Create Color Attribute=true
    Create SpriteSize Attribute=true
    SpriteSize=&Vector2DFromFloat(Value=&Multiply_Float(A=&RandomRangeFloat(Minimum=4, Maximum=5, Evaluation Type=::NewEnumerator1), B=$System.GlobalSize))
    Write to Module Namespace=true
   )
  ]
  rendererClasses [
   &NiagaraMeshRendererProperties
   &NiagaraSpriteRendererProperties
   &NiagaraLightRendererProperties
  ]
  renderers [
   {rendererClass=&NiagaraMeshRendererProperties, rendererIndex=0}
   {rendererClass=&NiagaraMeshRendererProperties, rendererIndex=1}
   {rendererClass=&NiagaraSpriteRendererProperties, rendererIndex=2}
   {rendererClass=&NiagaraLightRendererProperties, rendererIndex=3}
  ]
 }}
summary {
 systemName=NS_Mic_TaperSlink_Idle
 emitters [{
   emitterName=FlameInside
   rendererClasses [
    &NiagaraMeshRendererProperties
    &NiagaraSpriteRendererProperties
    &NiagaraLightRendererProperties
   ]
  }]
 userVariables [User.bIsFrontEndPreview {type=NiagaraBool}]
}
